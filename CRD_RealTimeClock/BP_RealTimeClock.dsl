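; Creative Real Time Clock device that broadcasts events when a configured UTC date is reached and when an optional following duration elapses.
; Knobs set the replicated target date and a timer; on firing while enabled, it advances the date by the repeat frequency and starts the duration timer.
; Dynamic materials display a countdown or the target date, with text opacity following enabled and play-mode state.
; Enabling also broadcasts whether the date has passed.
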(event Building|EventPlacedByPlacementTool
  ;[0 1 0] Initialization
  (Audio|PlaySoundatLocation (Variables|Audio|GetSoundPlaced) (Transformation|GetWorldLocation (Variables|Debug|GetStaticMeshComponent))))

(event EventBeginPlay
  ;[0 1 0] Initialization
  (|CallParentFunction)
  (Utilities|ConstructMIDs
    (:Loading%20Done
      (Default|SendTimeDatatoMaterial)
      (CallFunction|UpdateEnabledStateTextVisibility))
    (:Exec%20Out
      (Default|SetTargetDatefromKnobs)
      (Default|GetTimerHandle)
      (Default|RepeatDataReset))))

(event OnAnyPropertyChanged(ToyOptionsComponent)
  ;[0.224 0.7375 1] Real Time Clock
  (Utilities|ConstructMIDs
    (:Loading%20Done
      (Default|SendTimeDatatoMaterial)
      (CallFunction|UpdateEnabledStateTextVisibility))
    (:Exec%20Out
      (Default|SetTargetDatefromKnobs)
      (Default|GetTimerHandle)
      (Default|RepeatDataReset))))

(event OnPropertiesLoaded(ToyOptionsComponent)
  ;[0.224 0.7375 1] Real Time Clock
  (Utilities|ConstructMIDs
    (:Loading%20Done
      (Default|SendTimeDatatoMaterial)
      (CallFunction|UpdateEnabledStateTextVisibility))
    (:Exec%20Out
      (Default|SetTargetDatefromKnobs)
      (Default|GetTimerHandle)
      (Default|RepeatDataReset))))

(event Custom|OnRepeatingUpdatedTargetDate_Client
  ;[0.224 0.7375 1] Real Time Clock
  (Networking|CanExecuteCosmeticEvents
    (:True
      (Utilities|ConstructMIDs
        (:Loading%20Done
          (Default|SendTimeDatatoMaterial))))))

(event EventOnMinigameStarted
  ;[0.224 0.7375 1] Real Time Clock
  (Class|CreativeEnabledComponent|SetEnabled (Variables|Default|GetEnabledComponent) (Variables|UserKnobs|GetEnabledOnMinigameStart)))

(event MinigameStateChangedDelegate(FortMinigameProgress)
  (params (Minigame :type Fort%20Minigame%20Object%20Reference) (NewMinigameState :type EFortMinigameState%20Enum :default "PreGame"))
  ;[0.224 0.7375 1] Real Time Clock
  (CallFunction|UpdateEnabledStateTextVisibility))

(event Custom|EnableFunction
  (params (Instigator :type Controller%20Object%20Reference))
  ;[0.6588 0.4353 0.1373] Receivers
  (Class|CreativeEnabledComponent|SetEnabled (Variables|Default|GetEnabledComponent) true))

(event OnGameplayMessageReceived(EnableWhenReceived)
  (params (TriggerInstigator :type Controller%20Object%20Reference))
  ;[0.6588 0.4353 0.1373] Receivers
  (Class|CreativeEnabledComponent|SetEnabled (Variables|Default|GetEnabledComponent) true))

(event Custom|DisableFunction
  (params (Instigator :type Controller%20Object%20Reference))
  ;[0.6588 0.4353 0.1373] Receivers
  (Class|CreativeEnabledComponent|SetEnabled (Variables|Default|GetEnabledComponent) false))

(event OnGameplayMessageReceived(DisableWhenReceived)
  (params (TriggerInstigator :type Controller%20Object%20Reference))
  ;[0.6588 0.4353 0.1373] Receivers
  (Class|CreativeEnabledComponent|SetEnabled (Variables|Default|GetEnabledComponent) false))

(event Custom|OnTargetDateReached_Event
  ;[0.6588 0.4353 0.1373] Receivers
  (if (Class|CreativeEnabledComponent|IsEnabled (Variables|Default|GetEnabledComponent))
    (GameplayEvent|BroadcastEvent (Variables|UserOptions|Events|GetOn%20Time%20Reached) 0 (Utilities|Struct|MakeRTCTargetDateReachedData 0))
    (Default|RepeatDataCalculateonTargetDateReached)
    (CallFunction|ClampDurationtoTargetDate)
    (Default|GetDurationHandle)))

(event Custom|OnDurationExpired_Event
  ;[0.6588 0.4353 0.1373] Receivers
  (if (Class|CreativeEnabledComponent|IsEnabled (Variables|Default|GetEnabledComponent))
    (GameplayEvent|BroadcastEvent (Variables|UserOptions|Events|GetOn%20Duration%20Elapsed) 0 (Utilities|Struct|MakeRTCwhenDurationElapsedData 0))
    (Utilities|Time|ClearandInvalidateTimerbyHandle (Variables|Default|GetDurationHandle))))

(event OnEnabledStateChanged(EnabledComponent)
  (params (Enabled :type Boolean))
  ;[0.6588 0.4353 0.1373] Receivers
  (CallFunction|LogicwhenEnabledandDateTransmitOn Enabled)
  (CallFunction|UpdateEnabledStateTextVisibility))

(event EventEndPlay
  (params (EndPlayReason :type EEndPlayReason%20Enum :default "Destroyed"))
  ;[1 0.5 0] Cleanup
  (|CallParentFunction EndPlayReason)
  (Utilities|Time|ClearandInvalidateTimerbyHandle (Variables|Default|GetTimerHandle))
  (Utilities|Time|ClearandInvalidateTimerbyHandle (Variables|Default|GetDurationHandle)))

(fn CalculateTargetDateforMonthsRepeatFrequency
  (return (Math|DateTime|MakeDateTime (+ (Math|DateTime|GetYear (Math|DateTime|UTCNow)) (/ (Utilities|Operators|Add (Variables|UserKnobs|GetRepeatFrequency) (Math|DateTime|GetMonth (Math|DateTime|UTCNow)) -1) (Variables|Constants|GetMonthsPerYear))) (+ (Math|Integer|%25%28Integer%29 (Utilities|Operators|Add (Variables|UserKnobs|GetRepeatFrequency) (Math|DateTime|GetMonth (Math|DateTime|UTCNow)) -1) (Variables|Constants|GetMonthsPerYear)) 1) (Math|Integer|Min(Integer) (Math|DateTime|DaysInMonth (+ (Math|DateTime|GetYear (Math|DateTime|UTCNow)) (/ (Utilities|Operators|Add (Variables|UserKnobs|GetRepeatFrequency) (Math|DateTime|GetMonth (Math|DateTime|UTCNow)) -1) (Variables|Constants|GetMonthsPerYear))) (+ (Math|Integer|%25%28Integer%29 (Utilities|Operators|Add (Variables|UserKnobs|GetRepeatFrequency) (Math|DateTime|GetMonth (Math|DateTime|UTCNow)) -1) (Variables|Constants|GetMonthsPerYear)) 1)) (Math|DateTime|GetDay (Math|DateTime|UTCNow))) (Math|DateTime|GetHour (Math|DateTime|UTCNow)) (Math|DateTime|GetMinute (Math|DateTime|UTCNow)) (Math|DateTime|GetSecond (Math|DateTime|UTCNow)) (Math|DateTime|GetMillisecond (Math|DateTime|UTCNow)))))

(fn ClampDurationToTargetDate
  (bind _durationhandle (Variables|Default|GetDurationHandle))
  (if (Utilities|Time|IsValidTimerHandle _durationhandle)
    (if (> (Utilities|Time|GetTimerRemainingTimebyHandle _durationhandle) 0)
      (CallFunction|OnDurationExpiredEvent))))

(fn GetLocalizedTexture
  (bind _culturetexturestructarray (Variables|Materials|GetCultureTextureStructArray))
  (for _array_element _culturetexturestructarray
    (if (Utilities|String|EqualExactly(String) (Utilities|Internationalization|GetCurrentCulture) (.Culture_5_43F44D3C431D4902A2A77E8536B50255 (Utilities|Struct|BreakRealTimeClockCultureTextureStruct _array_element)))
      (return (Utilities|Struct|BreakRealTimeClockCultureTextureStruct _array_element))))
  (return (Utilities|Struct|BreakRealTimeClockCultureTextureStruct (Utilities|Array|Get(aref) _culturetexturestructarray (Variables|Materials|GetEnglishIndex)))))

(fn GetDurationHandle
  (Utilities|Time|ClearandInvalidateTimerbyHandle (Variables|Default|GetDurationHandle))
  (switch Utilities|FlowControl|Switch|SwitchonE_Device_RealTimeClock_TimeWindowDurationType (Variables|UserKnobs|GetTimeWindowDurationType)
    (:NewEnumerator0
      (return 0))
    (:NewEnumerator1
      (LocalVariables|SetCalculatedDurationTime (* (Variables|Constants|GetSecondsPerMinute) (Variables|UserKnobs|GetTimeWindowDurationValue))))
    (:NewEnumerator2
      (LocalVariables|SetCalculatedDurationTime (* (* (Variables|Constants|GetSecondsPerMinute) (Variables|Constants|GetMinutesPerHour)) (Variables|UserKnobs|GetTimeWindowDurationValue))))
    (:NewEnumerator3
      (LocalVariables|SetCalculatedDurationTime (* (* (* (Variables|Constants|GetSecondsPerMinute) (Variables|Constants|GetMinutesPerHour)) (Variables|Constants|GetHoursPerDay)) (Variables|UserKnobs|GetTimeWindowDurationValue)))))
  (bind _returnvalue (Utilities|Time|SetTimerbyEvent (EventDispatchers|CreateEvent :!event OnDurationExpired_Event) (Math|Conversions|ToFloat(Integer) (LocalVariables|GetCalculatedDurationTime))))
  (bind _output_get (Variables|Default|SetDurationHandle _returnvalue))
  (return _output_get))

(fn GetOverrideMeshMaterial
  (params (MeshComp :type Static%20Mesh%20Component%20Object%20Reference) (MatIdx :type Integer))
  (bind _realtimeclock (Variables|Default|GetRealTimeClock))
  (bind _realtimeclocknumbers (Variables|Default|GetRealTimeClockNumbers))
  (if (Utilities|Equal(Object) MeshComp _realtimeclock)
    (return (Utilities|Select :Index MatIdx :Option%200 (Rendering|Material|GetMaterial _realtimeclock MatIdx) :Option%201 (Variables|Materials|GetMIDClockText)))
    (else
      (if (Utilities|Equal(Object) MeshComp _realtimeclocknumbers)
        (return (Utilities|Select :Index MatIdx :Option%200 (Variables|Materials|GetMIDClockNumbers) :Option%201 (Rendering|Material|GetMaterial _realtimeclocknumbers MatIdx) :Option%202 (Variables|Materials|GetMIDClockText)))
        (else
          (return (|CallParentFunction MeshComp MatIdx)))))))

(fn GetTimerHandle
  (bind _returnvalue (Math|DateTime|UTCNow))
  (bind _targetdate (Variables|Default|GetTargetDate))
  (bind _returnvalue_2 (Math|DateTime|DateTime-DateTime _targetdate _returnvalue))
  (bind _returnvalue_3 (Math|Timespan|GetTotalSeconds _returnvalue_2))
  (Utilities|Time|ClearandInvalidateTimerbyHandle (Variables|Default|GetTimerHandle))
  (if (> _returnvalue_3 0.0)
    (bind _returnvalue_4 (Utilities|Time|SetTimerbyEvent (EventDispatchers|CreateEvent :!event OnTargetDateReached_Event) _returnvalue_3))
    (bind _output_get (Variables|Default|SetTimerHandle _returnvalue_4))
    (return _output_get)
    (else
      (return 0))))

(fn IsTargetableByLinearMover
  (return true))

(fn Logic_WhenEnabledAndDateTransmitOn
  (params (Enabled%20State :type Boolean))
  (Networking|SwitchHasAuthority
    (:Authority
      (if Enabled%20State
        (if (<= (Math|Timespan|GetTotalMilliseconds (Math|DateTime|DateTime-DateTime (Variables|Default|GetTargetDate) (Math|DateTime|UTCNow))) 0.0)
          (GameplayEvent|BroadcastEvent (Variables|UserOptions|Events|GetOn%20Enabling%20After%20Time%20Reached) 0 (Utilities|Struct|MakeRTCwhenEnabledandDateReachedData 0))
          (else
            (GameplayEvent|BroadcastEvent (Variables|UserOptions|Events|GetOn%20Enabling%20Before%20Time%20Reached) 0 (Utilities|Struct|MakeRTCwhenEnabledandDateNotReachedData 0))))))))

(fn OnRep_OnRepeatingUpdatedTargetDate)

(fn OnRep_TargetDate
  (CallFunction|OnRepeatingUpdatedTargetDateClient))

(fn Repeat_DataCalculateOnTargetDateReached
  (bind _repeatsremaining (Variables|Default|GetRepeatsRemaining))
  (Networking|SwitchHasAuthority
    (:Authority
      (if (Utilities|IsInfiniteRepeat _repeatsremaining)
        _
        (elif (> _repeatsremaining 0)
          (Math|Integer|DecrementInt (Variables|Default|GetRepeatsRemaining)
            (:%20%20
              _))))
      (switch Utilities|FlowControl|Switch|SwitchonE_Device_RealTimeClock_RepeatType (Variables|UserKnobs|GetRepeatType)
        (:NewEnumerator0
          (Variables|Default|SetTargetDate (Math.DateTime|AddTimetoDate (Math|DateTime|UTCNow) (* (Math|Timespan|MakeTimespan 0 1) (Variables|UserKnobs|GetRepeatFrequency)))))
        (:NewEnumerator1
          (Variables|Default|SetTargetDate (Math.DateTime|AddTimetoDate (Math|DateTime|UTCNow) (* (Math|Timespan|MakeTimespan 1) (Variables|UserKnobs|GetRepeatFrequency)))))
        (:NewEnumerator2
          (Variables|Default|SetTargetDate (Math.DateTime|AddTimetoDate (Math|DateTime|UTCNow) (* (Math|Timespan|MakeTimespan 7) (Variables|UserKnobs|GetRepeatFrequency)))))
        (:NewEnumerator3
          (bind _newdatetime (Default|CalculateTargetDateForMonthsRepeatFrequency))
          (Variables|Default|SetTargetDate _newdatetime)))
      (Default|GetTimerHandle))))

(fn Repeat_DataReset
  (Variables|Default|SetRepeatsRemaining (Variables|UserKnobs|GetRepeatNumber)))

(fn SendTimeDataToMaterial
  (bind _targetdate (Variables|Default|GetTargetDate))
  (bind (_year _month _day _hour _minute _second) (Math|DateTime|BreakDateTime _targetdate))
  (bind _midclocknumbers (Variables|Materials|GetMIDClockNumbers))
  (switch Utilities|FlowControl|Switch|SwitchonE_Device_RealTimeClock_DisplayMode (Variables|UserKnobs|GetDisplayMode)
    (:NewEnumerator0
      (Rendering|Material|SetScalarParameterValue _midclocknumbers "EndTime" (+ (Math|Timespan|GetTotalSeconds (Math|DateTime|DateTime-DateTime (Variables|Default|GetTargetDate) (Math|DateTime|UTCNow))) (Utilities|Time|GetRealTimeSeconds)) :!class /Script/Engine.MaterialInstanceDynamic)
      (Rendering|Material|SetScalarParameterValue _midclocknumbers "Number Day" 0.0 :!class /Script/Engine.MaterialInstanceDynamic)
      (Rendering|Material|SetScalarParameterValue _midclocknumbers "Number Hrs" 0.0 :!class /Script/Engine.MaterialInstanceDynamic)
      (Rendering|Material|SetScalarParameterValue _midclocknumbers "Number Min" 0.0 :!class /Script/Engine.MaterialInstanceDynamic)
      (Rendering|Material|SetScalarParameterValue _midclocknumbers "Number Sec" 0.0 :!class /Script/Engine.MaterialInstanceDynamic))
    (:NewEnumerator1
      (Rendering|Material|SetScalarParameterValue _midclocknumbers "EndTime" -1.0 :!class /Script/Engine.MaterialInstanceDynamic) ; Flag EndTime so we can set a custom value.
      (Rendering|Material|SetScalarParameterValue _midclocknumbers "Number Day" (Math|Conversions|ToFloat(Integer) _day) :!class /Script/Engine.MaterialInstanceDynamic)
      (Rendering|Material|SetScalarParameterValue _midclocknumbers "Number Hrs" (Math|Conversions|ToFloat(Integer) _hour) :!class /Script/Engine.MaterialInstanceDynamic)
      (Rendering|Material|SetScalarParameterValue _midclocknumbers "Number Min" (Math|Conversions|ToFloat(Integer) _minute) :!class /Script/Engine.MaterialInstanceDynamic)
      (Rendering|Material|SetScalarParameterValue _midclocknumbers "Number Sec" (Math|Conversions|ToFloat(Integer) _second) :!class /Script/Engine.MaterialInstanceDynamic))))

(fn SetTargetDateFromKnobs
  (Default|ValidateTargetDate)
  (Variables|Default|SetTargetDate (Math|DateTime|MakeDateTime (Variables|UserKnobs|GetTimeYear) (Variables|UserKnobs|GetTimeMonth) (Variables|UserKnobs|GetTimeDay) (Variables|UserKnobs|GetTimeHour) (Variables|UserKnobs|GetTimeMinute))))

(fn ShouldInvisiblePropsBeAllowed
  (return false))

(fn UpdateEnabledStateTextVisibility
  (bind _enabledcomponent (Variables|Default|GetEnabledComponent))
  (bind _enabled (Class|CreativeEnabledComponent|IsEnabled _enabledcomponent))
  (bind _returnvalue (Math|Float|SelectFloat 1.0 0.0 _enabled))
  (Networking|CanExecuteCosmeticEvents
    (:True
      (bind _currentminigame (Variables|/Script/FortniteGame.FortMinigameLogicComponent|GetCurrentMinigame (Variables|Default|GetFortMinigameProgress))
        (:then
          (Utilities|SetClockTextOpacity (Math|Float|SelectFloat _returnvalue (Math|Float|SelectFloat 1.0 (Variables|Materials|GetOpacityOfClockTextWhenDisabledInEditMode) (Variables|UserKnobs|GetEnabledOnMinigameStart)) (Class|CreativeCommonDeviceFunctionLibrary|IsinPlayMode _currentminigame))))
        (:else
          (Utilities|SetClockTextOpacity _returnvalue))) ; If we are in a minigame, use IsInPlayMode to determine if we should fully hide the clock\ntext when disabled or just grey it out so it can be easier to edit.\nOtherwise if minigame isn't valid, just use the enabled state of the device.
    )))

(fn ConstructionScript
  (|CallParentFunction))

(fn ValidateTargetDate
  (bind _output_get (LocalVariables|SetLargestLegalDay (Math|DateTime|DaysInMonth (Variables|UserKnobs|GetTimeYear) (Variables|UserKnobs|GetTimeMonth))))
  (if (< _output_get (Variables|UserKnobs|GetTimeDay))
    (Variables|UserKnobs|SetTimeDay (LocalVariables|GetLargestLegalDay))
    (OptionData|RefreshPropertyOverrides (Variables|Default|GetToyOptionsComponent))))
