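; Creative Real Time Clock device that fires events at a user-set real-world date, after an optional duration, and on enabling before or after that date.
; Options become a timezone-adjusted UTC target driving a timer, recalculated on the authority for day, week or month repeats.
; Replicated target changes refresh a dynamic material showing either a material-driven countdown or the date.
; A version upgrade migrates legacy repeat-count values to a separate enum.
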(event Building|EventPlacedByPlacementTool
  ;[0 1 0] Initialization
  (Audio|PlaySoundatLocation (Variables|Audio|GetSoundPlaced) (Transformation|GetWorldLocation (Variables|Debug|GetStaticMeshComponent))))

(event EventBeginPlay
  ;[0 1 0] Initialization
  (|CallParentFunction)
  (Utilities|ConstructMIDs
    (:Loading%20Done
      (Default|SendTimeDatatoMaterial)
      (CallFunction|UpdateEnabledStateTextVisibility))
    (:Exec%20Out
      (Default|SetTargetDatefromKnobs)
      (Default|GetTimerHandle)
      (Default|RepeatDataReset))))

(event OnAnyPropertyChanged(ToyOptionsComponent)
  ;[0.224 0.7375 1] Real Time Clock
  (Utilities|ConstructMIDs
    (:Loading%20Done
      (Default|SendTimeDatatoMaterial)
      (CallFunction|UpdateEnabledStateTextVisibility))
    (:Exec%20Out
      (Default|SetTargetDatefromKnobs)
      (Default|GetTimerHandle)
      (Default|RepeatDataReset))))

(event OnPropertiesLoaded(ToyOptionsComponent)
  ;[0.224 0.7375 1] Real Time Clock
  (Utilities|ConstructMIDs
    (:Loading%20Done
      (Default|SendTimeDatatoMaterial)
      (CallFunction|UpdateEnabledStateTextVisibility))
    (:Exec%20Out
      (Default|SetTargetDatefromKnobs)
      (Default|GetTimerHandle)
      (Default|RepeatDataReset))))

(event Custom|OnRepeatingUpdatedTargetDate_Client
  ;[0.224 0.7375 1] Real Time Clock
  (Networking|CanExecuteCosmeticEvents
    (:True
      (Utilities|ConstructMIDs
        (:Loading%20Done
          (Default|SendTimeDatatoMaterial))))))

(event EventOnMinigameStarted
  ;[0.224 0.7375 1] Real Time Clock
  (Class|CreativeEnabledComponent|SetEnabled (Variables|Default|GetEnabledComponent) (Variables|UserOptions|GetEnabledAtGameStart)))

(event MinigameStateChangedDelegate(FortMinigameProgress)
  (params (Minigame :type Fort%20Minigame%20Object%20Reference) (NewMinigameState :type EFortMinigameState%20Enum :default "PreGame"))
  ;[0.224 0.7375 1] Real Time Clock
  (CallFunction|UpdateEnabledStateTextVisibility))

(event Custom|EnableFunction
  (params (Instigator :type Controller%20Object%20Reference))
  ;[0.6588 0.4353 0.1373] Receivers
  (Class|CreativeEnabledComponent|SetEnabled (Variables|Default|GetEnabledComponent) true))

(event Custom|DisableFunction
  (params (Instigator :type Controller%20Object%20Reference))
  ;[0.6588 0.4353 0.1373] Receivers
  (Class|CreativeEnabledComponent|SetEnabled (Variables|Default|GetEnabledComponent) false))

(event Custom|OnTargetDateReached_Event
  (bind _on_time_reached (Variables|UserOptions|Events|GetOn%20Time%20Reached))
  ;[0.6588 0.4353 0.1373] Receivers
  (if (Class|CreativeEnabledComponent|IsEnabled (Variables|Default|GetEnabledComponent))
    (bind (_selectedinstigatorslist _validinstigatoradded) (CallFunction|GetBroadcastInstigators))
    (if _validinstigatoradded
      (for _array_element _selectedinstigatorslist
        (GameplayEvent|BroadcastEvent _on_time_reached 0 (Utilities|Struct|MakeRTCTargetDateReachedData _array_element)))
      (else
        (GameplayEvent|BroadcastEvent _on_time_reached 0 (Utilities|Struct|MakeRTCTargetDateReachedData 0))))
    (Default|RepeatDataCalculateonTargetDateReached)
    (CallFunction|ClampDurationtoTargetDate)
    (Default|GetDurationHandle)))

(event Custom|OnDurationExpired_Event
  ;[0.6588 0.4353 0.1373] Receivers
  (if (Class|CreativeEnabledComponent|IsEnabled (Variables|Default|GetEnabledComponent))
    (bind (_selectedinstigatorslist _validinstigatoradded) (CallFunction|GetBroadcastInstigators))
    (if _validinstigatoradded
      (for _array_element _selectedinstigatorslist
        (GameplayEvent|BroadcastEvent (Variables|UserOptions|Events|GetOn%20Duration%20Elapsed) 0 (Utilities|Struct|MakeRTCwhenDurationElapsedData _array_element)))
      (else
        (GameplayEvent|BroadcastEvent (Variables|UserOptions|Events|GetOn%20Duration%20Elapsed) 0 (Utilities|Struct|MakeRTCwhenDurationElapsedData 0))))
    (Utilities|Time|ClearandInvalidateTimerbyHandle (Variables|Default|GetDurationHandle))))

(event OnEnabledStateChanged(EnabledComponent)
  (params (Enabled :type Boolean))
  ;[0.6588 0.4353 0.1373] Receivers
  (CallFunction|LogicwhenEnabledandDateTransmitOn Enabled)
  (CallFunction|UpdateEnabledStateTextVisibility))

(event EventEndPlay
  (params (EndPlayReason :type EEndPlayReason%20Enum :default "Destroyed"))
  ;[1 0.5 0] Cleanup
  (|CallParentFunction EndPlayReason)
  (Utilities|Time|ClearandInvalidateTimerbyHandle (Variables|Default|GetTimerHandle))
  (Utilities|Time|ClearandInvalidateTimerbyHandle (Variables|Default|GetDurationHandle)))

(event HandleVersionUpgrades(CreativeVersion)
  (params (LastSavedVersion :type Creative%20Version%20Id%20Structure))
  (CallFunction|HandleVersionUpgrade LastSavedVersion))

(fn CalculateTargetDateforMonthsRepeatFrequency
  (return (Math|DateTime|MakeDateTime (+ (Math|DateTime|GetYear (Math|DateTime|UTCNow)) (/ (Utilities|Operators|Add (Variables|UserOptions|Basic|GetRepeatFrequency) (Math|DateTime|GetMonth (Math|DateTime|UTCNow)) -1) (Variables|Constants|GetMonthsPerYear))) (+ (Math|Integer|%25%28Integer%29 (Utilities|Operators|Add (Variables|UserOptions|Basic|GetRepeatFrequency) (Math|DateTime|GetMonth (Math|DateTime|UTCNow)) -1) (Variables|Constants|GetMonthsPerYear)) 1) (Math|Integer|Min(Integer) (Math|DateTime|DaysInMonth (+ (Math|DateTime|GetYear (Math|DateTime|UTCNow)) (/ (Utilities|Operators|Add (Variables|UserOptions|Basic|GetRepeatFrequency) (Math|DateTime|GetMonth (Math|DateTime|UTCNow)) -1) (Variables|Constants|GetMonthsPerYear))) (+ (Math|Integer|%25%28Integer%29 (Utilities|Operators|Add (Variables|UserOptions|Basic|GetRepeatFrequency) (Math|DateTime|GetMonth (Math|DateTime|UTCNow)) -1) (Variables|Constants|GetMonthsPerYear)) 1)) (Math|DateTime|GetDay (Math|DateTime|UTCNow))) (Math|DateTime|GetHour (Math|DateTime|UTCNow)) (Math|DateTime|GetMinute (Math|DateTime|UTCNow)) (Math|DateTime|GetSecond (Math|DateTime|UTCNow)) (Math|DateTime|GetMillisecond (Math|DateTime|UTCNow)))))

(fn ClampDurationToTargetDate
  (bind _durationhandle (Variables|Default|GetDurationHandle))
  (if (Utilities|Time|IsValidTimerHandle _durationhandle)
    (if (> (Utilities|Time|GetTimerRemainingTimebyHandle _durationhandle) 0)
      (CallFunction|OnDurationExpiredEvent))))

(fn FilterBotsFromFortPlayerStates
  (params (InFortPlayerStates :type Array%20of%20Fort%20Player%20State%20Object%20References))
  ;[0 1 0.0938 0.5] Returns a copy of the input array without bots.
  (for _array_element InFortPlayerStates
    (Utilities|IsValid _array_element
      (:Is%20Valid
        (if (not (Variables|/Script/Engine.PlayerState|GetbIsABot _array_element))
          (Utilities|Array|Add (LocalVariables|GetNonBotPlayersLocal) _array_element)))))
  (return (LocalVariables|GetNonBotPlayersLocal)))

(fn GetLocalizedTexture
  (bind _culturetexturestructarray (Variables|Materials|GetCultureTextureStructArray))
  (for _array_element _culturetexturestructarray
    (if (Utilities|String|EqualExactly(String) (Utilities|Internationalization|GetCurrentCulture) (.Culture_5_43F44D3C431D4902A2A77E8536B50255 (Utilities|Struct|BreakRealTimeClockCultureTextureStruct _array_element)))
      (return (Utilities|Struct|BreakRealTimeClockCultureTextureStruct _array_element))))
  (return (Utilities|Struct|BreakRealTimeClockCultureTextureStruct (Utilities|Array|Get(aref) _culturetexturestructarray (Variables|Materials|GetEnglishIndex)))))

(fn GetBroadcastInstigators
  (bind _self self)
  (bind _returnvalue (Minigame|GetMinigame _self :!class /Script/FortniteGame.MinigameBlueprintLibrary))
  (bind _outplayers (Minigame|GetParticipatingPlayers _returnvalue))
  (bind _nonbotfortplayerstates (Default|FilterBotsfromFortPlayerStates _outplayers))
  (bind (_outitem _outindex) (Utilities|Array|RandomArrayItem _nonbotfortplayerstates))
  (Utilities|FlowControl|Sequence
    (:then_0
      (switch Utilities|FlowControl|Switch|SwitchonECreativeInstigator (Variables|UserOptions|GetInstigatorForEvent)
        (:NewEnumerator0
          (return 0 false))
        (:NewEnumerator1
          (if (>= _outindex 0)
            (bind _ascontroller (Utilities|Casting|CastToController (Actor|GetOwner _outitem)))
            (Utilities|Array|Add (LocalVariables|GetSelectedInstigators) _ascontroller)))
        (:NewEnumerator2
          (for _array_element _nonbotfortplayerstates
            (bind _ascontroller_2 (Utilities|Casting|CastToController (Actor|GetOwner _array_element)))
            (Utilities|Array|Add (LocalVariables|GetSelectedInstigators) _ascontroller_2)))))
    (:then_1
      (return (LocalVariables|GetSelectedInstigators) (Utilities|Array|IsNotEmpty (LocalVariables|GetSelectedInstigators))))))

(fn GetDurationHandle
  (Utilities|Time|ClearandInvalidateTimerbyHandle (Variables|Default|GetDurationHandle))
  (switch Utilities|FlowControl|Switch|SwitchonE_Device_RealTimeClock_TimeWindowDurationType (Variables|UserOptions|Basic|GetDurationType)
    (:NewEnumerator0
      (return 0))
    (:NewEnumerator1
      (LocalVariables|SetCalculatedDurationTime (* (Variables|Constants|GetSecondsPerMinute) (Variables|UserOptions|Basic|GetDurationValue))))
    (:NewEnumerator2
      (LocalVariables|SetCalculatedDurationTime (* (* (Variables|Constants|GetSecondsPerMinute) (Variables|Constants|GetMinutesPerHour)) (Variables|UserOptions|Basic|GetDurationValue))))
    (:NewEnumerator3
      (LocalVariables|SetCalculatedDurationTime (* (* (* (Variables|Constants|GetSecondsPerMinute) (Variables|Constants|GetMinutesPerHour)) (Variables|Constants|GetHoursPerDay)) (Variables|UserOptions|Basic|GetDurationValue)))))
  (bind _returnvalue (Utilities|Time|SetTimerbyEvent (EventDispatchers|CreateEvent :!event OnDurationExpired_Event) (Math|Conversions|ToFloat(Integer) (LocalVariables|GetCalculatedDurationTime))))
  (bind _output_get (Variables|Default|SetDurationHandle _returnvalue))
  (return _output_get))

(fn GetOverrideMeshMaterial
  (params (MeshComp :type Static%20Mesh%20Component%20Object%20Reference) (MatIdx :type Integer))
  (bind _realtimeclock (Variables|Default|GetRealTimeClock))
  (bind _realtimeclocknumbers (Variables|Default|GetRealTimeClockNumbers))
  (if (Utilities|Equal(Object) MeshComp _realtimeclock)
    (return (Utilities|Select :Index MatIdx :Option%200 (Rendering|Material|GetMaterial _realtimeclock MatIdx) :Option%201 (Variables|Materials|GetMIDClockText)))
    (else
      (if (Utilities|Equal(Object) MeshComp _realtimeclocknumbers)
        (return (Utilities|Select :Index MatIdx :Option%200 (Variables|Materials|GetMIDClockNumbers) :Option%201 (Rendering|Material|GetMaterial _realtimeclocknumbers MatIdx) :Option%202 (Variables|Materials|GetMIDClockText)))
        (else
          (return (|CallParentFunction MeshComp MatIdx)))))))

(fn GetTimerHandle
  (bind _returnvalue (Math|DateTime|UTCNow))
  (bind _targetdate (Variables|Default|GetTargetDate))
  (bind _returnvalue_2 (Math|DateTime|DateTime-DateTime _targetdate _returnvalue))
  (bind _returnvalue_3 (Math|Timespan|GetTotalSeconds _returnvalue_2))
  (Utilities|Time|ClearandInvalidateTimerbyHandle (Variables|Default|GetTimerHandle))
  (if (> _returnvalue_3 0.0)
    (bind _returnvalue_4 (Utilities|Time|SetTimerbyEvent (EventDispatchers|CreateEvent :!event OnTargetDateReached_Event) _returnvalue_3))
    (bind _output_get (Variables|Default|SetTimerHandle _returnvalue_4))
    (return _output_get)
    (else
      (return 0))))

(fn HandleVersionUpgrade
  (params (LastSavedVersion :type Creative%20Version%20Id%20Structure%20%28by%20ref%29))
  (bind _numberofrepeats (Variables|Default|GetNumberOfRepeats))
  (Utilities|FlowControl|Sequence
    (:then_0
      ;[0.3036 0.6858 0.6682 0.5] Replace alias values in NumberOfRepeats with separate enum that hides irrelevant options related to repeating events.
      (CreativeVersion|CompareVersiontoTag (Variables|GetLastSavedVersion) "(TagName=\"Creative.Version.Release.38.00\")"
        (:OlderVersion
          (if (== _numberofrepeats -1)
            (Variables|UserOptions|Basic|SetEventRepeats "NewEnumerator1")
            (else
              (if (== _numberofrepeats 0)
                (Variables|UserOptions|Basic|SetEventRepeats "NewEnumerator0")
                (Variables|UserOptions|Basic|SetNumberOfRepeats_V2 1)
                (else
                  (Variables|UserOptions|Basic|SetEventRepeats "NewEnumerator2")
                  (Variables|UserOptions|Basic|SetNumberOfRepeats_V2 (Variables|Default|GetNumberOfRepeats))))
              (LocalVariables|SetMigrated true))))))
    (:then_1
      (if (LocalVariables|GetMigrated)
        (OptionData|RefreshPropertyOverrides (Variables|Default|GetToyOptionsComponent))))))

(fn IsInfiniteRepeat
  (return (Utilities|Enum|Equal(Enum) (Variables|UserOptions|Basic|GetEventRepeats) "NewEnumerator1")))

(fn Logic_WhenEnabledAndDateTransmitOn
  (params (Enabled%20State :type Boolean))
  (bind _on_enabling_after_time_reached (Variables|UserOptions|Events|GetOn%20Enabling%20After%20Time%20Reached))
  (bind _on_enabling_before_time_reached (Variables|UserOptions|Events|GetOn%20Enabling%20Before%20Time%20Reached))
  (Networking|SwitchHasAuthority
    (:Authority
      (if Enabled%20State
        (if (<= (Math|Timespan|GetTotalMilliseconds (Math|DateTime|DateTime-DateTime (Variables|Default|GetTargetDate) (Math|DateTime|UTCNow))) 0.0)
          (bind (_selectedinstigatorslist _validinstigatoradded) (CallFunction|GetBroadcastInstigators))
          (if _validinstigatoradded
            (for _array_element _selectedinstigatorslist
              (GameplayEvent|BroadcastEvent _on_enabling_after_time_reached 0 (Utilities|Struct|MakeRTCwhenEnabledandDateReachedData _array_element)))
            (else
              (GameplayEvent|BroadcastEvent _on_enabling_after_time_reached 0 (Utilities|Struct|MakeRTCwhenEnabledandDateReachedData 0))))
          (else
            (bind (_selectedinstigatorslist_2 _validinstigatoradded_2) (CallFunction|GetBroadcastInstigators))
            (if _validinstigatoradded_2
              (for _array_element_2 _selectedinstigatorslist_2
                (GameplayEvent|BroadcastEvent _on_enabling_before_time_reached 0 (Utilities|Struct|MakeRTCwhenEnabledandDateNotReachedData _array_element_2)))
              (else
                (GameplayEvent|BroadcastEvent _on_enabling_before_time_reached 0 (Utilities|Struct|MakeRTCwhenEnabledandDateNotReachedData 0))))))))))

(fn OnRep_OnRepeatingUpdatedTargetDate)

(fn OnRep_TargetDate
  (CallFunction|OnRepeatingUpdatedTargetDateClient))

(fn Repeat_DataCalculateOnTargetDateReached
  (Networking|SwitchHasAuthority
    (:Authority
      (if (CallFunction|IsInfiniteRepeat)
        _
        (elif (> (Variables|Default|GetRepeatsRemaining) 0)
          (Math|Integer|DecrementInt (Variables|Default|GetRepeatsRemaining)
            (:%20%20
              _))))
      (switch Utilities|FlowControl|Switch|SwitchonE_Device_RealTimeClock_RepeatType (Variables|UserOptions|Basic|GetRepeatType)
        (:NewEnumerator0
          (Variables|Default|SetTargetDate (Math.DateTime|AddTimetoDate (Math|DateTime|UTCNow) (* (Math|Timespan|MakeTimespan 0 1) (Variables|UserOptions|Basic|GetRepeatFrequency)))))
        (:NewEnumerator1
          (Variables|Default|SetTargetDate (Math.DateTime|AddTimetoDate (Math|DateTime|UTCNow) (* (Math|Timespan|MakeTimespan 1) (Variables|UserOptions|Basic|GetRepeatFrequency)))))
        (:NewEnumerator2
          (Variables|Default|SetTargetDate (Math.DateTime|AddTimetoDate (Math|DateTime|UTCNow) (* (Math|Timespan|MakeTimespan 7) (Variables|UserOptions|Basic|GetRepeatFrequency)))))
        (:NewEnumerator3
          (bind _newdatetime (Default|CalculateTargetDateForMonthsRepeatFrequency))
          (Variables|Default|SetTargetDate _newdatetime)))
      (Default|GetTimerHandle))))

(fn Repeat_DataReset
  (switch Utilities|FlowControl|Switch|SwitchonE_Device_RealTimeClock_RepeatFrequency (Variables|UserOptions|Basic|GetEventRepeats)
    (:NewEnumerator0
      (Variables|Default|SetRepeatsRemaining 0))
    (:NewEnumerator1)
    (:NewEnumerator2
      (Variables|Default|SetRepeatsRemaining (Variables|UserOptions|Basic|GetNumberOfRepeats_V2)))))

(fn SendTimeDataToMaterial
  (bind _utctimezone (Variables|UserOptions|Advanced|GetUTCTimezone))
  (bind _localtimezonedifferences (Variables|Default|GetLocalTimezoneDifferences))
  (bind _value (Utilities|Map|Find _localtimezonedifferences _utctimezone))
  (bind _returnvalue (Math|Timespan|MakeTimespan 0 _value))
  (bind _targetdate (Variables|Default|GetTargetDate))
  (bind _returnvalue_2 (Math|DateTime|DateTime+Timespan _targetdate _returnvalue))
  (bind (_year _month _day _hour _minute _second) (Math|DateTime|BreakDateTime _returnvalue_2))
  (bind _midclocknumbers (Variables|Materials|GetMIDClockNumbers))
  (switch Utilities|FlowControl|Switch|SwitchonE_Device_RealTimeClock_DisplayMode (Variables|UserOptions|GetDisplayMode)
    (:NewEnumerator0
      (Rendering|Material|SetScalarParameterValue _midclocknumbers "EndTime" (+ (Math|Timespan|GetTotalSeconds (Math|DateTime|DateTime-DateTime (Variables|Default|GetTargetDate) (Math|DateTime|UTCNow))) (Utilities|Time|GetRealTimeSeconds)) :!class /Script/Engine.MaterialInstanceDynamic)
      (Rendering|Material|SetScalarParameterValue _midclocknumbers "Number Day" 0.0 :!class /Script/Engine.MaterialInstanceDynamic)
      (Rendering|Material|SetScalarParameterValue _midclocknumbers "Number Hrs" 0.0 :!class /Script/Engine.MaterialInstanceDynamic)
      (Rendering|Material|SetScalarParameterValue _midclocknumbers "Number Min" 0.0 :!class /Script/Engine.MaterialInstanceDynamic)
      (Rendering|Material|SetScalarParameterValue _midclocknumbers "Number Sec" 0.0 :!class /Script/Engine.MaterialInstanceDynamic))
    (:NewEnumerator1
      (Rendering|Material|SetScalarParameterValue _midclocknumbers "EndTime" -1.0 :!class /Script/Engine.MaterialInstanceDynamic) ; Flag EndTime so we can set a custom value.
      (Rendering|Material|SetScalarParameterValue _midclocknumbers "Number Day" (Math|Conversions|ToFloat(Integer) _day) :!class /Script/Engine.MaterialInstanceDynamic)
      (Rendering|Material|SetScalarParameterValue _midclocknumbers "Number Hrs" (Math|Conversions|ToFloat(Integer) _hour) :!class /Script/Engine.MaterialInstanceDynamic)
      (Rendering|Material|SetScalarParameterValue _midclocknumbers "Number Min" (Math|Conversions|ToFloat(Integer) _minute) :!class /Script/Engine.MaterialInstanceDynamic)
      (Rendering|Material|SetScalarParameterValue _midclocknumbers "Number Sec" (Math|Conversions|ToFloat(Integer) _second) :!class /Script/Engine.MaterialInstanceDynamic))))

(fn SetTargetDateFromKnobs
  (Default|ValidateTargetDate)
  ;[1 1 1] If the user is using UTC+0 don't alter value to match UTC+0.
  (if (Utilities|Enum|Equal(Enum) (Variables|UserOptions|Advanced|GetUTCTimezone) "NewEnumerator12")
    (Variables|Default|SetTargetDate (Math|DateTime|MakeDateTime (Variables|UserOptions|Basic|GetYear) (Variables|UserOptions|Basic|GetMonth) (Variables|UserOptions|Basic|GetDay) (Variables|UserOptions|Basic|GetHour) (Variables|UserOptions|Basic|GetMinute)))
    (else
      (Variables|Default|SetTargetDate (Math|DateTime|DateTime+Timespan (Math|DateTime|MakeDateTime (Variables|UserOptions|Basic|GetYear) (Variables|UserOptions|Basic|GetMonth) (Variables|UserOptions|Basic|GetDay) (Variables|UserOptions|Basic|GetHour) (Variables|UserOptions|Basic|GetMinute)) (Math|Timespan|MakeTimespan 0 (* (Utilities|Map|Find (Variables|Default|GetLocalTimezoneDifferences) (Variables|UserOptions|Advanced|GetUTCTimezone)) -1)))))))

(fn UpdateEnabledStateTextVisibility
  (bind _enabledcomponent (Variables|Default|GetEnabledComponent))
  (bind _enabled (Class|CreativeEnabledComponent|IsEnabled _enabledcomponent))
  (bind _returnvalue (Math|Float|SelectFloat 1.0 0.0 _enabled))
  (Networking|CanExecuteCosmeticEvents
    (:True
      (bind _currentminigame (Variables|/Script/FortniteGame.FortMinigameLogicComponent|GetCurrentMinigame (Variables|Default|GetFortMinigameProgress))
        (:then
          (Utilities|SetClockTextOpacity (Math|Float|SelectFloat _returnvalue (Math|Float|SelectFloat 1.0 (Variables|Materials|GetOpacityOfClockTextWhenDisabledInEditMode) (Variables|UserOptions|GetEnabledAtGameStart)) (Class|CreativeCommonDeviceFunctionLibrary|IsinPlayMode _currentminigame))))
        (:else
          (Utilities|SetClockTextOpacity _returnvalue))) ; If we are in a minigame, use IsInPlayMode to determine if we should fully hide the clock\ntext when disabled or just grey it out so it can be easier to edit.\nOtherwise if minigame isn't valid, just use the enabled state of the device.
    )))

(fn ConstructionScript
  (|CallParentFunction))

(fn ValidateTargetDate
  (bind _output_get (LocalVariables|SetLargestLegalDay (Math|DateTime|DaysInMonth (Variables|UserOptions|Basic|GetYear) (Variables|UserOptions|Basic|GetMonth))))
  (if (< _output_get (Variables|UserOptions|Basic|GetDay))
    (Variables|UserOptions|Basic|SetDay (LocalVariables|GetLargestLegalDay))
    (OptionData|RefreshPropertyOverrides (Variables|Default|GetToyOptionsComponent))))
